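# Animation Blueprint for the M_MED_MoonSatellite tail cosmetic, targeting the shared Fortnite base tail skeleton.
# Its only listed graph is an empty EventGraph, with no variables and no animation logic shown.
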AnimBlueprint /BRCosmetics/Accessories/FORT_Tails/Tail_M_MED_MoonSatellite/Meshes/M_MED_MoonSatellite_Tail_AnimBP
target_skeleton=/Game/Accessories/FORT_Tails/Common/Fortnite_Base_Tail_Skeleton
graphs [EventGraph {links [], nodes []}]
variables []
